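# M_ChocolateChunks is a masked, two-sided lit material for mesh particles, apparently chocolate chunks in the LoveLlama melee weapon effect.
# Particle color tints the base color parameter, with scalar parameters setting metallic, specular and roughness.
# An optional erosion mask feeds a smoke texture, its UVs scaled by particle red, into MF_Dissolve, driven by inverse particle alpha.
# Static switches choose emissive from a time-of-day-weighted tinted mask, a two-color Fresnel or plain intensity.
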Material /FNGameplayCues/Weapons/FORT_Melee/Effects/LoveLlama/Materials/M_ChocolateChunks
graph {
 properties {bUsedWithMeshParticles=true, blendMode=BLEND_Masked, pixelDepthOffsetMode=PDOM_Legacy, twoSided=true, usedShadingModels=Default Lit}
 nodes [
  n1 = ParticleColor()
  Multiply(A=VectorParameter() {parameterName=BaseColor, defaultValue=<rgba 1 1 1 1>}.RGB, B=@n1.RGB)
  ScalarParameter() {parameterName=Metallic}
  ScalarParameter() {defaultValue=0.25, parameterName=Specular}
  ScalarParameter() {defaultValue=0.25, parameterName=Roughness}
  n2 = Multiply(A=TextureCoordinate(), B=@n1.R)
  StaticSwitchParameter(
   True=MaterialFunctionCall!(
    Ramp=TextureSampleParameter2D(UVs=@n2) {parameterName=ErodeTexture, sortPriority=3, texture=&T_Inky_Smoke_Tile}.R
    Dissolve=OneMinus(None=@n1.A)
    EdgeSharpness
   ) {materialFunction=/Game/Effects/Fort_Effects/Materials/_Functions/MF_Dissolve}.ResultWithOutClamp
   False=Constant() {r=1}
  ) {defaultValue=true, parameterName=UseErode}
  n3 = ParticleColor()
  n4 = Multiply(
   A=TextureSampleParameter2D() {parameterName=EmissiveMask, texture=/Game/Environments/Props/DinosaurPark/Textures/T_Dinosaur_Park_Brontosaurus_Shell_D}.RGB
   B=@n3.RGB
  )
  n5 = Constant() {r=1}
  n6 = Constant()
  n7 = MaterialFunctionCall(Color_Morning=@n5, Color_Day=@n5, Color_Evening=@n6, Color_Night=@n6) {materialFunction=&Color_Time_of_Day}
  n8 = ScalarParameter() {parameterName=EmissiveIntensity}
  n9 = StaticSwitchParameter(True=Multiply(A=@n3.A, B=@n8), False=@n8) {parameterName=UseParticleColorToControlEmissiveIntensity}
  n10 = LinearInterpolate(
   A=VectorParameter() {parameterName=LowColor}.RGB
   B=VectorParameter() {parameterName=HighColor}.RGB
   Alpha=Fresnel(
    ExponentIn=ScalarParameter() {defaultValue=5, parameterName=ExponentIn}
    BaseReflectFractionIn=ScalarParameter() {parameterName=BaseReflectFractionIn}
   )
  )
  StaticSwitchParameter(
   True=LinearInterpolate(B=Multiply(A=@n4, B=@n7.Result), Alpha=ScalarParameter() {parameterName=EmissiveIntensity})
   False=StaticSwitchParameter(True=Multiply(A=@n9, B=@n10), False=@n8) {parameterName=UseFresnelEmissive}
  ) {parameterName=UseParticleColorControlEmissive}
 ]
}
